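# Dark metal material in the glass screen effects folder, using Default Lit shading and enabled for Nanite.
# It holds a constant dark grey color and a high scalar constant.
# The red channel of an atmospheric cloud noise texture blends between two scalar values, adding cloudy variation.
# The file does not show which material inputs these values feed.
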Material /SparksMaps/Effects/Screens/Glass/Materials/Dark_Metal_A
graph {
 properties {bUsedWithNanite=true, pixelDepthOffsetMode=PDOM_Legacy, usedShadingModels=Default Lit}
 nodes [
  Constant4Vector() {constant=<rgba 0.1 0.1 0.1 0>}
  Constant() {r=0.85}
  LinearInterpolate(Alpha=TextureSample() {texture=&AtmosphericCloudNoise01}.R) {constA=0.25, constB=0.4}
 ]
}
